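# Default-lit material for the M80 gadget, enabled for use on mesh particles.
# A single texture sample feeds ConvertFromDiffSpec as diffuse color, alongside a constant low specular value.
# The same texture, dimmed by a constant multiply, feeds the shared MF_FX_EmissiveAdjustments function as emissive.
# A red color constant and a scalar constant sit unconnected in the graph.
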Material /Game/Gadgets/Assets/M80/M_M80
graph {
 properties {bUsedWithMeshParticles=true, pixelDepthOffsetMode=PDOM_Legacy, usedShadingModels=Default Lit}
 nodes [
  Constant3Vector() {constant=<rgba 1 0 0 0>}
  Constant() {r=0.1}
  n1 = TextureSample() {texture=&/T_M80}
  MaterialFunctionCall(Emissive=Multiply(A=@n1.RGB, B=Constant() {r=0.2})) {materialFunction=&MF_FX_EmissiveAdjustments}
  MaterialFunctionCall(DiffuseColor=@n1.RGB, SpecularColor=Constant() {r=0.2}) {materialFunction=&ConvertFromDiffSpec}
 ]
}
